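# EQS query that picks a sprite discoverable, favoring nearby, high-priority candidates.
# A custom generator, which cannot run async, supplies sprite discoverables, filtered by safe-zone and can-be-discovered tests.
# Scoring combines inverse-linear 3D distance from the querier with a linear priority test that also filters.
# Tests are auto-sorted.
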EnvQuery /Sprites/Gameplay/EQS/EQS_SpriteDiscoverable
options [{
  generator {
   name=FortQueryGenerator_SpriteDiscoverables_0
   class=FortQueryGenerator_SpriteDiscoverables
   properties {bAutoSortTests=true, bCanRunAsync=false, envQueryResultNormalizationOption=Default, optionName=""}
  }
  tests [
   FortQueryTest_InsideAthenaSafeZone_0 {class=FortQueryTest_InsideAthenaSafeZone, properties {testPurpose=Filter}}
   EnvQueryTest_Distance_0 {
    class=EnvQueryTest_Distance
    properties {
     distanceTo=&EnvQueryContext_Querier
     filterType=Range
     floatValueMin=0
     scoringEquation=InverseLinear
     scoringFactor=1
     testComment=""
     testMode=Distance3D
     testPurpose=Score
    }
   }
   FortQueryTest_SpriteDiscoverablePriority_0 {
    class=FortQueryTest_SpriteDiscoverablePriority
    properties {
     bDefineReferenceValue=false
     boolValue=true
     clampMaxType=None
     clampMinType=None
     filterType=Range
     floatValueMax=0
     floatValueMin=0
     multipleContextFilterOp=AllPass
     multipleContextScoreOp=AverageScore
     normalizationType=Absolute
     referenceValue=0
     scoreClampMax=0
     scoreClampMin=0
     scoringEquation=Linear
     scoringFactor=1
     testComment=""
     testPurpose=FilterAndScore
    }
   }
   FortQueryTest_SpriteDiscoverableCanBeDiscovered_0 {
    class=FortQueryTest_SpriteDiscoverableCanBeDiscovered
    properties {
     bDefineReferenceValue=false
     boolValue=true
     clampMaxType=None
     clampMinType=None
     filterType=Match
     floatValueMax=0
     floatValueMin=0
     multipleContextFilterOp=AllPass
     multipleContextScoreOp=AverageScore
     normalizationType=Absolute
     referenceValue=0
     scoreClampMax=0
     scoreClampMin=0
     scoringEquation=Constant
     scoringFactor=1
     testComment=""
     testPurpose=Filter
    }
   }
  ]
 }]
